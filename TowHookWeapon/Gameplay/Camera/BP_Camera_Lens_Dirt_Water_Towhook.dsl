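; Camera lens effect actor for the TowHook weapon that, going by its name, shows dirt and water on the screen.
; On BeginPlay it finds player 0's FortPawn and, if that pawn has a valid TetheredMovement component, binds to its OnWaterStateChanged dispatcher.
; The handler passes the in-water bool to the Niagara component as "bIsOnWater", leaving any visual change to the Niagara system.
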(event EventBeginPlay
  (bind _asfort_pawn (Utilities|Casting|CastToFortPawn (Game|GetPlayerPawn 0)))
  (Utilities|IsValid (Actor|GetComponentbyClass _asfort_pawn "/Script/TowHookWeaponRuntime.FortPawnComponent_TetheredMovement")
    (:Is%20Valid
      (EventDispatchers|BindEventtoOnWaterStateChanged (Actor|GetComponentbyClass _asfort_pawn "/Script/TowHookWeaponRuntime.FortPawnComponent_TetheredMovement") (EventDispatchers|CreateEvent :!event Custom|OnWater)))))

(event Custom|OnWater
  (params (bInWater :type Boolean))
  (Niagara|SetNiagaraVariable(Bool) (Variables|NiagaraActor|GetNiagaraComponent) "bIsOnWater" bInWater))

(fn ConstructionScript)
